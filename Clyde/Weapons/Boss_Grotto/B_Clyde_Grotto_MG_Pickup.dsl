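; Pickup Blueprint for the Clyde Grotto boss MG weapon that plays the item's pickup sound when collected.
; On pickup it runs the parent logic, reads the item definition's pickup data and async loads the PickupSound soft reference if valid.
; For a locally viewed picker, it spawns the sound at the pickup's root and registers that audio component with the MusicEventSubsystem as a first-person ability event.
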(event EventOnPickedUp
  (params (PickupTarget :type Fort%20Pawn%20Object%20Reference))
  (|CallParentFunction PickupTarget)
  (bind _itemdata (Class|ItemDefinitionBase|GetItemData (Variables|PickupProperties|GetItemDefinition))
    (:Found
      (if (Utilities|IsValidSoftObjectReference (.PickupSound (Utilities|Struct|BreakFortnitePickupItemData _itemdata)))
        (bind _object (Utilities|AsyncLoadAsset (.PickupSound (Utilities|Struct|BreakFortnitePickupItemData _itemdata)))
          (:Completed
            (bind _assound_base (Utilities|Casting|CastToSoundBase _object))
            (if (Pawn|IsLocallyViewed PickupTarget)
              (bind _returnvalue (Audio|SpawnSoundatLocation _assound_base (Transformation|GetWorldLocation (Variables|Default|GetRoot))))
              (Audio|AddEventbyComponent (GameInstanceSubsystems|GetMusicEventSubsystem) _returnvalue "(TagName=\"MusicEventSubsystem.Event.Gameplay.Ability.1P\")" "(GameplayTags=)" 0 ""))))))))

(fn ConstructionScript
  (|CallParentFunction))
