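; Minigun weapon Blueprint that plays a vocal track and muzzle lyric effects while firing, with lyrics hidden under one licensed-audio setting.
; For non-boss wielders, overheat sets the track start point; the local client shares it to sync others.
; A MetaSound output watcher feeds lyric time and global time dilation to the muzzle system.
; A hotfix-gated local timer targets ForceOverheat before the track loops; stopping fire clears it and silences the vocals.
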(event Custom|ForceOverheat)

(event EventOnPlayWeaponFireFX
  (params (bPersistentFire :type Boolean) (bSecondaryFire :type Boolean))
  (bind _returnvalue (Game|GetInstigator))
  (bind _returnvalue_2 (Pawn|IsLocallyControlled _returnvalue))
  (bind _overheat_scalar (Variables|Default|GetOverheat_Scalar))
  (bind _returnvalue_3 (CallFunction|CurrentScalarValue))
  (bind _returnvalue_4 (Math|Float|SelectFloat _returnvalue_3 _overheat_scalar _returnvalue_2))
  (bind _voxsoundwave (Variables|Default|GetVoxSoundWave))
  (bind _duration (Variables|/Script/Engine.SoundBase|GetDuration _voxsoundwave))
  (bind _returnvalue_5 (* _returnvalue_4 _duration))
  (bind _returnvalue_6 (Math|Float|Max(Float) _returnvalue_5))
  (bind _spawned_muzzle_system (Variables|Effects|Muzzle|GetSpawned%20Muzzle%20System))
  (|CallParentFunction bPersistentFire bSecondaryFire)
  (CallFunction|SendGlobalTimedialtoMuzzleSystem)
  (bind _returnvalue_7 (Fortnite|Audio|GetLicensedAudioTreatment))
  (if (== _returnvalue_7 2)
    (Effects|Components|ParticleSystem|SetBoolParameter _spawned_muzzle_system "ShowLyrics" false)
    (else
      (Effects|Components|ParticleSystem|SetBoolParameter _spawned_muzzle_system "ShowLyrics" true)))
  (if (not (Variables|Default|GetIsBoss))
    ;[1 1 1] Sync the track start point to the current OverheatValue. Makes lyrics represent how close we are to Overheated.
    (if _returnvalue_2
      (CallFunction|UpdateOverheatScalar _returnvalue_6) ; If OwningClient, send the Overheat value to other clients. Syncs the lyrics for everyone
    )
    (Variables|Default|SetVoxPlayEndTimeDif _returnvalue_6) ; Set the 'tracker' for where in the track we are along the total duration
    (if (and (ScalableFloat|GetValueasBool (Variables|Default|GetHotfixForceOverheat)) (Pawn|IsLocallyControlled (Game|GetInstigator)))
      ;[1 1 1] Failsafe to prevent Minigun track from Looping during heavy Server loads, low Client frame rate, or long RTT between Client & Server
      (bind _returnvalue_8 (Utilities|Time|SetTimerbyEvent (delegate Custom|ForceOverheat) (Math|Float|Max(Float) (- (Variables|/Script/Engine.SoundBase|GetDuration (Variables|Default|GetVoxSoundWave)) (Variables|Default|GetVoxPlayEndTimeDif))))) ; Set timer to predict when the weapon should overheat to prevent looping track
      (Variables|Default|SetForceOverheatTimer _returnvalue_8))
    (Audio|Parameter|SetFloatParameter (Variables|Default|GetMinigunVox) "StartTime" (Variables|Default|GetVoxPlayEndTimeDif) :!class /Script/Engine.AudioComponent)
    (Audio|Components|Audio|Play (Variables|Default|GetMinigunVox) 0.0)
    (Effects|Components|ParticleSystem|SetFloatParameter (Variables|Effects|Muzzle|GetSpawned%20Muzzle%20System) "StartTime" (Variables|Default|GetVoxPlayEndTimeDif))
    (Variables|Default|SetVoxPlayStartTime (Utilities|Time|GetGameTimeinSeconds))
    (bind _minigunvox (Variables|Default|GetMinigunVox)
      (:then
        (Audio|AddEventbyComponent (GameInstanceSubsystems|GetMusicEventSubsystem) _minigunvox (select (Variables|Default|GetbIsLocal) (Variables|Sound|GetMusic1PTag) (Variables|Sound|GetMusic3PTag)) "(GameplayTags=)" 0 ""))
      (:else))))

(event EventBeginPlay
  ;[0 0 0] audio to Niagara
  (|CallParentFunction)
  ;;[1 1 1] Set licensed track and lyric FX to loop
  (Audio|Parameter|SetBooleanParameter (Variables|Default|GetMinigunVox) "Loops" (not (ScalableFloat|GetValueasBool (Variables|Default|GetHotfixDisableRapGodLoop))) :!class /Script/Engine.AudioComponent)
  (bind _returnvalue (MetaSoundOutput|WatchOutput (WorldSubsystems|GetMetaSoundOutputSubsystem) (Variables|Default|GetMinigunVox) "NormalizedLyricTime" (delegate Custom|MetasoundAudioToNiagara) "None" "None" :!class /Script/MetasoundEngine.MetaSoundOutputSubsystem))
  (Development|PrintString (Utilities|String|ToString(Boolean) _returnvalue)))

(event Custom|MetasoundAudioToNiagara
  (params (OutputName :type Name) (Output :type Meta%20Sound%20Output%20Structure%20%28by%20ref%29))
  ;[0 0 0] audio to Niagara
  (bind (_success _returnvalue) (MetaSoundOutput|GetFloat Output))
  (Effects|Components|ParticleSystem|SetFloatParameter (Variables|Effects|Muzzle|GetSpawned%20Muzzle%20System) "MetasoundTime" _returnvalue)
  (CallFunction|SendGlobalTimedialtoMuzzleSystem))

(event Custom|SendGlobalTimedialtoMuzzleSystem
  (Effects|Components|ParticleSystem|SetFloatParameter (Variables|Effects|Muzzle|GetSpawned%20Muzzle%20System) "GlobalTimeDialationVal" (Utilities|Time|GetGlobalTimeDilation)))

(event EventOnStopWeaponFireFX
  (|CallParentFunction)
  (CallFunction|SendGlobalTimedialtoMuzzleSystem)
  (if (not (Variables|Default|GetIsBoss))
    (Audio|Components|Audio|FadeOut (Variables|Default|GetMinigunVox) 0.0 0.0 "Linear" :!class /Script/Engine.AudioComponent)
    (Utilities|Time|ClearandInvalidateTimerbyHandle (Variables|Default|GetForceOverheatTimer))
    (bind _minigunvox (Variables|Default|GetMinigunVox)
      (:then
        (Audio|RemoveEvent (GameInstanceSubsystems|GetMusicEventSubsystem) _minigunvox))
      (:else))))

(fn ConstructionScript
  (|CallParentFunction))

(fn WrapVoxTime
  (params (StartTime :type Float%20%28double-precision%29) (EndTime :type Float%20%28double-precision%29))
  (return (Math|Float|%25%28Float%29 (+ (- (Variables|GetEndTime) (Variables|GetStartTime)) (Variables|Default|GetVoxPlayEndTimeDif)) (Variables|/Script/Engine.SoundBase|GetDuration (Variables|Default|GetVoxSoundWave)))))
