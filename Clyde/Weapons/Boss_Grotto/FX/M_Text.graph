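# Translucent unlit Niagara sprite and mesh particle material that draws animated, outlined text from a distance-field lyrics flipbook for Boss Grotto weapon FX.
# Four panned samples of one normal map blend to distort flipbook UVs; smoothstep thresholds form the fill and outline.
# Particle color tints the fill, rotated smoke pattern varies brightness, and noise dissolves opacity.
# Four dynamic parameter sets expose phase, outline, distortion, dissolve, softness and color controls per particle.
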Material /Clyde/Weapons/Boss_Grotto/FX/M_Text
graph {
 properties {bUsedWithMeshParticles=true, bUsedWithNiagaraSprites=true, blendMode=BLEND_Translucent, shadingModel=MSM_Unlit, usedShadingModels=Unlit}
 nodes [
  LinearInterpolate(Alpha=MaterialFunctionCall() {materialFunction=&DebugTimeSine}.Result) {constA=1, constB=-2}
  n1 = MaterialFunctionCall(X=NamedRerouteUsage(), Y=NamedRerouteUsage(), Z=NamedRerouteUsage()) {materialFunction=&MakeFloat3}
  n2 = ParticleColor()
  n3 = ScalarParameter() {defaultValue=0.6, parameterName=Center}
  n4 = NamedRerouteUsage()
  n5 = Subtract(A=@n3, B=@n4) {constA=0.5}
  n6 = Add(A=@n4, B=@n3) {constB=0.5}
  n7 = NamedRerouteUsage()
  n8 = ScalarParameter() {defaultValue=10, parameterName="Number of Rows (S)"}
  n9 = ScalarParameter() {defaultValue=10, parameterName="Number of Columns (S)"}
  n10 = VectorParameter() {parameterName=Distort Tile, channelNames {a=NA, b=Tile XY, g=Tile Y, r=Tile X}, defaultValue=<rgba 1 1 1 0>}
  n11 = Multiply(A=TextureCoordinate(), B=Multiply(A=AppendVector(A=@n10.Tile X, B=@n10.Tile X), B=@n10.Tile XY))
  n12 = Multiply(A=ScalarParameter() {defaultValue=1, parameterName=SPeed}, B=NamedRerouteUsage())
  n13 = TextureObjectParameter() {
   parameterName=Distortion Normal
   samplerType=SAMPLERTYPE_Normal
   texture=/Game/Accessories/FORT_Backpacks/Backpack_M_MED_EctoCat/FX/CloudNoise_T_N
  }
  n14 = ScalarParameter() {defaultValue=3, parameterName=Normal Mip Bias}
  n15 = TextureSample!(UVs=Panner(Coordinate=@n11, Time=@n12) {speedX=0.1, speedY=0.1}, Tex=@n13, Bias=@n14, Apply View MipBias) {mipValueMode=TMVM_MipBias, samplerType=SAMPLERTYPE_Normal, texture=&CloudNoise_T_N}
  n16 = Constant2Vector() {g=0.3548, r=0.4181}
  n17 = TextureSample(UVs=Panner(Coordinate=Add(A=@n11, B=@n16), Time=@n12) {speedX=-0.1, speedY=-0.1}, Tex=@n13, Bias=@n14) {mipValueMode=TMVM_MipBias, samplerType=SAMPLERTYPE_Normal, texture=&CloudNoise_T_N}
  n18 = Constant2Vector() {g=0.1484, r=0.8649}
  n19 = TextureSample(UVs=Panner(Coordinate=Add(A=@n11, B=@n18), Time=@n12) {speedX=-0.1, speedY=0.1}, Tex=@n13, Bias=@n14) {mipValueMode=TMVM_MipBias, samplerType=SAMPLERTYPE_Normal, texture=&CloudNoise_T_N}
  n20 = Constant2Vector() {g=0.7516, r=0.6513}
  n21 = TextureSample(UVs=Panner(Coordinate=Add(A=@n11, B=@n20), Time=@n12) {speedX=0.1, speedY=-0.1}, Tex=@n13, Bias=@n14) {mipValueMode=TMVM_MipBias, samplerType=SAMPLERTYPE_Normal, texture=&CloudNoise_T_N}
  n22 = ComponentMask(None=MaterialFunctionCall(
    BaseNormal=MaterialFunctionCall(BaseNormal=@n15.RGB, AdditionalNormal=@n17.RGB) {materialFunction=&BlendAngleCorrectedNormals}.Result
    AdditionalNormal=MaterialFunctionCall(BaseNormal=@n19.RGB, AdditionalNormal=@n21.RGB) {materialFunction=&BlendAngleCorrectedNormals}.Result
   ) {materialFunction=&BlendAngleCorrectedNormals}.Result) {g=true, r=true}
  n23 = Multiply(A=NamedRerouteUsage()) {constB=0.01}
  n24 = TextureSampleParameter2D!(
   UVs=Add(
    A=MaterialFunctionCall("Animation  Phase (0-1)"=@n7, Number of Rows=@n8, Number of Columns=@n9) {materialFunction=&FlipBook}.UVs
    B=Multiply(A=@n22, B=@n23)
   )
   Level=Constant()
   Apply View MipBias
  ) {
   mipValueMode=TMVM_MipLevel
   parameterName=Param
   samplerType=SAMPLERTYPE_DistanceFieldFont
   texture=&/T_ClayPlug_Lyrics_Text
  }
  n25 = SmoothStep(Min=@n5, Max=@n6, Value=@n24.R) {constMax=0.54, constMin=0.5}
  n26 = Saturate(None=@n25)
  n27 = TextureCoordinate()
  n28 = NamedRerouteUsage()
  n29 = LinearInterpolate(
   A=ScalarParameter() {defaultValue=0.5, parameterName=Color Mul Min}
   B=ScalarParameter() {defaultValue=1, parameterName=Color Mul Max}
   Alpha=TextureSample!(
    UVs=MaterialFunctionCall(UVs=@n27, "Rotation Angle (0-1)"=@n28) {materialFunction=&CustomRotator}.Rotated Values
    Tex
    Apply View MipBias
   ) {texture=/Game/Global/Textures/Default/Tileable_Patterns/T_Blowing_Smoke_Pattern2_Tile_DISP}.R
  )
  n30 = NamedRerouteUsage()
  MaterialFunctionCall(Emissive=Multiply(A=LinearInterpolate(A=@n1.Result, B=@n2.RGB, Alpha=@n26), B=LinearInterpolate(B=@n29, Alpha=@n30) {constA=1})) {materialFunction=&MF_FX_EmissiveAdjustments}
  n31 = Add(A=@n24.R, B=NamedRerouteUsage()) {constB=0.05}
  n32 = Saturate(None=Add(A=@n25, B=SmoothStep(Min=@n5, Max=@n6, Value=@n31) {constMax=0.54, constMin=0.5}))
  n33 = Constant() {r=1}
  n34 = NamedRerouteUsage()
  n35 = NamedRerouteUsage()
  n36 = Multiply(A=Add(A=TextureCoordinate(), B=AppendVector(A=@n34, B=@n35)), B=AppendVector(A=NamedRerouteUsage(), B=NamedRerouteUsage()))
  n37 = VectorParameter() {parameterName=Disolve texture Options, channelNames {a=contrast, b=add, g=Mul, r=pow}, defaultValue=<rgba 1 1 0 0>}
  n38 = Multiply(
   A=Power(
    Base=TextureSample(UVs=@n36) {samplerType=SAMPLERTYPE_LinearGrayscale, texture=/Game/Effects/Fort_Effects/Textures/T_DarkAndLight_Noise}.R
    Exp=@n37.pow
   )
   B=@n37.Mul
  )
  n39 = Saturate(None=MaterialFunctionCall(In=Add(A=@n38, B=@n37.add)) {materialFunction=&CheapContrast}.Result)
  n40 = Constant()
  n41 = NamedRerouteUsage()
  Multiply(
   A=Saturate(None=Multiply(A=@n2.A, B=@n32))
   B=Saturate(None=MaterialFunctionCall!(1=@n33, 2=@n39, 3=@n40, 4, A=@n41) {materialFunction=/Engine/Functions/Engine_MaterialFunctions02/Lerp_Multiple_Float}.Lerp 3 Inputs)
  )
  n42 = DynamicParameter() {defaultValue=<rgba 0 0.1 0 0>, paramNames [Anim Phase, Outline thickness, Distortion amount, Time]}
  NamedRerouteDeclaration(None=@n42.Time) {name=DP0_A_Time, nodeColor=<rgba 1 1 1 1>}
  NamedRerouteDeclaration(None=@n42.Anim Phase) {name=DP0_R_AnimPhase, nodeColor=<rgba 0 1 0.25 1>}
  n43 = DynamicParameter() {parameterIndex=1, defaultValue=<rgba 0 0 0 1>, paramNames [outline R, outline G, outline B, Dissolve]}
  NamedRerouteDeclaration(None=@n43.Dissolve) {name=DP1_A_Dissolve, nodeColor=<rgba 1 1 1 1>}
  n44 = DynamicParameter() {parameterIndex=2, paramNames [DissilveTileX, DissolveTileY, DissolveOffsetX, DissolveOffsetY]}
  NamedRerouteDeclaration(None=@n44.DissilveTileX) {name=DP2_R_DissolveTileX, nodeColor=<rgba 0 0.3906 1 1>}
  NamedRerouteDeclaration(None=@n44.DissolveTileY) {name=DP2_G_DissolveTileY, nodeColor=<rgba 0 1 0 1>}
  NamedRerouteDeclaration(None=@n42.Outline thickness) {name=DP0_G_OutlineThickness, nodeColor=<rgba 0 1 0 1>}
  NamedRerouteDeclaration(None=@n44.DissolveOffsetX) {name=DP2_B_DissolveOffsetX, nodeColor=<rgba 0 0 1 1>}
  NamedRerouteDeclaration(None=@n44.DissolveOffsetY) {name=DP2_A_DissolveOffsetY, nodeColor=<rgba 1 1 1 1>}
  n45 = DynamicParameter() {parameterIndex=3, defaultValue=<rgba 0.01 1 1 1>, paramNames [TextSmoothness, Text Color Mul, Color Rot, NA]}
  NamedRerouteDeclaration(None=@n45.TextSmoothness) {name=DP3_R_TextSoftness, nodeColor=<rgba 0.0078 0 1 1>}
  NamedRerouteDeclaration(None=@n45.Text Color Mul) {name=DP3_G_TextColorMul, nodeColor=<rgba 0 1 0 1>}
  NamedRerouteDeclaration(None=@n45.Color Rot) {name=DP3_B_ColorRot, nodeColor=<rgba 0 0 1 1>}
  NamedRerouteDeclaration(None=@n45.NA) {name=DP3_A_, nodeColor=<rgba 1 1 1 1>}
  NamedRerouteDeclaration(None=@n42.Distortion amount) {name=DP0_B_DistortionAmount, nodeColor=<rgba 0 0 1 1>}
  NamedRerouteDeclaration(None=@n43.outline R) {name=DP1_R_OutlineR, nodeColor=<rgba 0.4062 0 1 1>}
  NamedRerouteDeclaration(None=@n43.outline G) {name=DP1_G_OutlineG, nodeColor=<rgba 0 1 0 1>}
  NamedRerouteDeclaration(None=@n43.outline B) {name=DP1_B_OutlineB, nodeColor=<rgba 0 0 1 1>}
  ScalarParameter() {parameterName=Softness}
 ]
}
