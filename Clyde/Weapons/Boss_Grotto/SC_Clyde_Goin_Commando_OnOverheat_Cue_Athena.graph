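# Sound cue for the overheat of the Clyde Grotto boss "Goin Commando" weapon in Athena.
# A mixer layers a ballistic turret overheat sound with a music stinger boosted above it.
# The stinger sits behind a licensed-content switcher with only its third input connected, so it depends on that switch.
# Close weapon-projectile attenuation with falloff overrides spatializes it, and overridden concurrency prevents new instances at the limit.
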SoundCue /Clyde/Weapons/Boss_Grotto/SC_Clyde_Goin_Commando_OnOverheat_Cue_Athena
root=Mixer_0
nodes [
 {id=FortSoundNodeLicensedContentSwitcher_0, class=FortSoundNodeLicensedContentSwitcher, children [null,null,WavePlayer_0]}
 {id=Mixer_0, children [WavePlayer_2, FortSoundNodeLicensedContentSwitcher_0], properties {inputVolume [1,4]}}
 {id=WavePlayer_0, properties {soundWaveAssetPtr=/Clyde/Audio/Items/ClayPlug_Weapons/MX_IP_ClydeGrotto_MG_Overheat_Stinger}}
 {id=WavePlayer_2, properties {soundWaveAssetPtr=/Game/Sounds/Fort_Traps/MountedTurret/Turret_Ballistic_Overheat_03}}
]
properties {
 attenuationSettings=&Fort_Weapon_Projectile_Close_Athena_Att
 bOverrideConcurrency=true
 maxDistance=4950
 soundClassObject=&Fort_WeaponRanged_Fire_3P_PSM
 volumeMultiplier=0.3
 attenuationOverrides {distanceAlgorithm=NaturalSound, falloffDistance=4000, attenuationShapeExtents {x=1000}}
 concurrencyOverrides {resolutionRule=PreventNew}
}
