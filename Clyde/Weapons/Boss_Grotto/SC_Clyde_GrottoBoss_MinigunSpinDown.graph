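# Sound cue for the Grotto boss minigun spin-down, built from the Goin' Commando spin-down waves.
# A mixer layers a close wave and a distant wave, each with its own attenuation, a modulator and a third-person ranged-fire sound class override.
# The distant layer switches on quality level between the standard wave and an optimized variant.
# The mixer's third input is empty.
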SoundCue /Clyde/Weapons/Boss_Grotto/SC_Clyde_GrottoBoss_MinigunSpinDown
root=Mixer_0
nodes [
 {id=Attenuation_1, children [Modulator_2], properties {attenuationSettings=&Fort_Weapon_Projectile_Close_Athena_Att}}
 {
  id=Attenuation_2
  children [Modulator_3]
  properties {attenuationSettings=/Game/Sounds/Attenuation/Fort_Weapon_Projectile_Medium_MinigunSpin_Athena_Att}
 }
 {id=Mixer_0, children [SoundClass_1,SoundClass_2,null], properties {inputVolume [1,1,1]}}
 {id=Modulator_2, children [WavePlayer_2]}
 {id=Modulator_3, children [QualityLevel_0]}
 {id=QualityLevel_0, children [WavePlayer_3,WavePlayer_3,WavePlayer_4,WavePlayer_4]}
 {
  id=SoundClass_1
  children [Attenuation_1]
  properties {soundClassOverride=&Fort_WeaponRanged_Fire_3P_PSM}
 }
 {
  id=SoundClass_2
  children [Attenuation_2]
  properties {soundClassOverride=&Fort_WeaponRanged_Fire_3P_PSM}
 }
 {id=WavePlayer_2, properties {soundWaveAssetPtr=/Game/Sounds/Fort_Weapons/AutoRifle/WPN_GoinCommando_SpinDown_close_athena}}
 {id=WavePlayer_3, properties {soundWaveAssetPtr=/Game/Sounds/Fort_Weapons/AutoRifle/WPN_GoinCommando_SpinDown_distant_athena}}
 {id=WavePlayer_4, properties {soundWaveAssetPtr=/Game/Sounds/Fort_Weapons/AutoRifle/WPN_GoinCommando_SpinDown_distant_athena_opt}}
]
properties {maxDistance=4950, volumeMultiplier=1.5}
