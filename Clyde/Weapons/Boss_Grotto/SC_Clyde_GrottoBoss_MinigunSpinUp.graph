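# Sound cue for the Grotto boss minigun spin-up, built from close and distant GoinCommando spin-up recordings.
# Each recording runs through a modulator, its own attenuation for close or medium range, and a sound class override to the third-person ranged fire class.
# Both layers meet in a mixer with an unused third input.
# The cue sets a long rifle attenuation, boosts overall volume and leaves its debug flag on.
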SoundCue /Clyde/Weapons/Boss_Grotto/SC_Clyde_GrottoBoss_MinigunSpinUp
root=Mixer_17
nodes [
 {
  id=Attenuation_0
  children [Modulator_1]
  properties {attenuationSettings=/Game/Sounds/Attenuation/Fort_Weapon_Projectile_Medium_MinigunSpin_Athena_Att}
 }
 {id=Attenuation_49, children [Modulator_0], properties {attenuationSettings=&Fort_Weapon_Projectile_Close_Athena_Att}}
 {id=Mixer_17, children [SoundClass_32, SoundClass_0, null], properties {inputVolume [1,1,1]}}
 {id=Modulator_0, children [WavePlayer_10]}
 {id=Modulator_1, children [WavePlayer_1]}
 {
  id=SoundClass_0
  children [Attenuation_0]
  properties {soundClassOverride=&Fort_WeaponRanged_Fire_3P_PSM}
 }
 {
  id=SoundClass_32
  children [Attenuation_49]
  properties {soundClassOverride=&Fort_WeaponRanged_Fire_3P_PSM}
 }
 {id=WavePlayer_1, properties {soundWaveAssetPtr=/Game/Sounds/Fort_Weapons/AutoRifle/WPN_GoinCommando_SpinUp_distant_athena}}
 {id=WavePlayer_10, properties {soundWaveAssetPtr=/Game/Sounds/Fort_Weapons/AutoRifle/WPN_GoinCommando_SpinUp_close_athena}}
]
properties {
 attenuationSettings=/Game/Sounds/Attenuation/Fort_Weapon_Rifle_Lng_Att
 bDebug=true
 maxDistance=7500
 soundClassObject=&Fort_WeaponRanged
 volumeMultiplier=1.5
}
